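; Anim notify state that suppresses cosmetic companion footstep FX for the duration of an animation window.
; On begin, it finds the FortPlayerCompanionCosmeticComponent on the mesh owner and sets bFootstepFXDisabled to true; on end, it sets it back to false.
; Each event returns whether that component was found, and end clears the flag outright rather than restoring its prior value.
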(fn Received_NotifyBegin
  (params (MeshComp :type Skeletal%20Mesh%20Component%20Object%20Reference) (Animation :type Anim%20Sequence%20Base%20Object%20Reference) (TotalDuration :type Float%20%28single-precision%29) (EventReference :type Anim%20Notify%20Event%20Reference%20Structure%20%28by%20ref%29))
  (bind _returnvalue (Components|GetOwner MeshComp))
  (bind _returnvalue_2 (Actor|GetComponentbyClass _returnvalue "/Script/CosmeticCompanionsRuntime.FortPlayerCompanionCosmeticComponent"))
  (if (Utilities|IsValid _returnvalue_2 :!class /Script/Engine.KismetSystemLibrary)
    (Variables|/Script/CosmeticCompanionsRuntime.FortPlayerCompanionCosmeticComponent|SetbFootstepFXDisabled true _returnvalue_2)
    (return true)
    (else
      (return false))))

(fn Received_NotifyEnd
  (params (MeshComp :type Skeletal%20Mesh%20Component%20Object%20Reference) (Animation :type Anim%20Sequence%20Base%20Object%20Reference) (EventReference :type Anim%20Notify%20Event%20Reference%20Structure%20%28by%20ref%29))
  (bind _returnvalue (Components|GetOwner MeshComp))
  (bind _returnvalue_2 (Actor|GetComponentbyClass _returnvalue "/Script/CosmeticCompanionsRuntime.FortPlayerCompanionCosmeticComponent"))
  (if (Utilities|IsValid _returnvalue_2 :!class /Script/Engine.KismetSystemLibrary)
    (Variables|/Script/CosmeticCompanionsRuntime.FortPlayerCompanionCosmeticComponent|SetbFootstepFXDisabled false _returnvalue_2)
    (return true)
    (else
      (return false))))
