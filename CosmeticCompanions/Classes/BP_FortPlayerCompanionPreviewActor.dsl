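; Frontend preview actor for a player companion, playing spawn animations in the locker and chooser-selected reaction montages, including responses to the owner pawn's emotes.
; Camera, emote and member-readiness flags feed RefreshVisibility, which hides the actor, pauses animation and spawns appear or disappear VFX outside the locker.
; Frontend-presence changes toggle a hologram material, swapping and restoring cached overlay materials.
; The res-in effect drives a timeline dissolve, then reveals cosmetic VFX.
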(event EventBeginPlay
  (EventDispatchers|BindEventtoOnCosmeticsInitialized (Variables|FortPlayerCompanionPreviewActor|GetCosmeticComponent) (EventDispatchers|CreateEvent :!event Custom|OnCosmeticsInitialised))
  (bind _asfort_pawn (Utilities|Casting|CastToFortPawn (Game|GetInstigator)))
  (EventDispatchers|AssignOnPawnStartedEmote _asfort_pawn (EventDispatchers|CreateEvent :!event Custom|OnPawnStartedEmote_Event))
  (EventDispatchers|BindEventtoOnPawnStoppedEmote _asfort_pawn (EventDispatchers|CreateEvent :!event Custom|OnPawnStoppedEmote_Event)))

(event Custom|OnCosmeticsInitialised
  (Variables|Default|SetInitState "NewEnumerator1")
  (if (Variables|FortPlayerCompanionPreviewActor|GetbIsLockerActor)
    ;[1 1 1] Hide the Cosmetic VFX to show them during the PlayResInEffect
    (Class|FortPlayerCompanionCosmeticComponent|SetCosmeticsVFXVisiblity (Variables|FortPlayerCompanionPreviewActor|GetCosmeticComponent) false)
    (Utilities|FlowControl|RetriggerableDelay (Variables|Default|GetSpawnAnimDelay))
    (if (and (GameplayTags|HasTag (Variables|FortPlayerCompanionPreviewActor|GetContextTags) "(TagName=\"UI.ItemPreview.PreviewType.OnPawn\")") (Utilities|Enum|Equal(Enum) (Variables|FortPlayerCompanionPreviewActor|GetSpawnAnimMode) "Default"))
      (Variables|FortPlayerCompanionPreviewActor|SetSpawnAnimMode "None") ; Default spawn anim should not play on pawn, but emotes and interacts still should.
    )
    (Networking|CanExecuteCosmeticEvents
      (:True
        (CallFunction|PlaySpawnAnim)))))

(event Custom|OnPawnStartedEmote_Event
  (params (MontageItemDef :type Fort%20Item%20Definition%20Object%20Reference) (PawnEmoting :type Fort%20Pawn%20Object%20Reference))
  (bind _returnvalue (GameplayTags|GetOwnedGameplayTags MontageItemDef :!class /Script/GameplayTags.BlueprintGameplayTagLibrary))
  (bind _output_get (Variables|Default|SetEmoteHidden (GameplayTags|HasTag _returnvalue "(TagName=\"Cosmetics.Emote.HideMimosaInPreview\")")))
  (CallFunction|RefreshVisibility)
  (if (not _output_get)
    (CallFunction|PlayReaction (select (GameplayTags|HasAnyTags _returnvalue (GameplayTags|MakeLiteralGameplayTagContainer "(GameplayTags=((TagName=\"Cosmetics.UserFacingFlags.Emote.Dance\"),(TagName=\"Cosmetics.UserFacingFlags.Emote.Sparks\"),(TagName=\"Cosmetics.Emote.MimosaReaction.Dance\")))")) (Variables|Default|GetEmoteReactionTagDance) (Variables|Default|GetEmoteReactionTagNonDance)))))

(event Custom|OnPawnStoppedEmote_Event
  (params (MontageItemDef :type Fort%20Item%20Definition%20Object%20Reference) (PawnEmoting :type Fort%20Pawn%20Object%20Reference))
  (Variables|Default|SetEmoteHidden false)
  (CallFunction|RefreshVisibility))

(event Custom|RefreshVisibility
  (bind (_origin _cylinderhalfheight) (Utilities|Struct|BreakFortUICameraFrameTargetBounds (Variables|ItemPreview|GetCameraFrameTargetBounds)))
  (switch Utilities|FlowControl|Switch|SwitchonECompanionPreviewInitState (Variables|Default|GetInitState)
    (:NewEnumerator1
      _)
    (:NewEnumerator2
      (if (!= (Math|Boolean|ORBoolean (Variables|Default|GetDiscoveryHomeHidden) (Variables|Default|GetEmoteHidden) (Variables|Default|GetOwnerGameReadinessHidden)) (Variables|Default|GetIsHidden))
        _)))
  (bind _output_get (Variables|Default|SetIsHidden (Math|Boolean|ORBoolean (Variables|Default|GetDiscoveryHomeHidden) (Variables|Default|GetEmoteHidden) (Variables|Default|GetOwnerGameReadinessHidden))))
  (if _output_get
    (CallFunction|StopReactionAnim)
    (else
      _))
  (Rendering|SetActorHiddenInGame _output_get)
  (Components|SkeletalMesh|SetEnableAnimation (Variables|FortPlayerCompanionPreviewActor|GetMesh) (not _output_get))
  (Utilities|FlowControl|Sequence
    (:then_0
      (if (and (or (Utilities|Enum|Equal(Enum) (Variables|Default|GetInitState) "NewEnumerator2") (not (Variables|Default|GetIsHidden))) (not (Variables|FortPlayerCompanionPreviewActor|GetbIsLockerActor)))
        (Effects|Components|ParticleSystem|SpawnFXAtLocation (select (Variables|Default|GetIsHidden) (Variables|Default|GetDisappearVFX) (Variables|Default|GetReappearVFX)) (Utilities|Operators|Add (Transformation|GetActorLocation) (Math|Vector|MakeVector 0.0 0.0 (* _cylinderhalfheight 0.5))))))
    (:then_1
      (Variables|Default|SetInitState "NewEnumerator2"))))

(event EventEndPlay
  (params (EndPlayReason :type EEndPlayReason%20Enum :default "Destroyed"))
  (bind _asfort_pawn (Utilities|Casting|CastToFortPawn (Game|GetInstigator)))
  (EventDispatchers|UnbindEventfromOnPawnStartedEmote _asfort_pawn (EventDispatchers|CreateEvent :!event Custom|OnPawnStartedEmote_Event))
  (EventDispatchers|UnbindEventfromOnPawnStoppedEmote _asfort_pawn (EventDispatchers|CreateEvent :!event Custom|OnPawnStoppedEmote_Event)))

(event EventPlayReaction_Internal
  (params (ReactionTag :type Gameplay%20Tag%20Structure%20%28by%20ref%29) (CallbackWrapper :type Fort%20Player%20Companion%20Preview%20Actor%20Reaction%20Finished%20Delegate%20Wrapper%20Object%20Reference))
  (bind _cosmeticcomponent (Variables|FortPlayerCompanionPreviewActor|GetCosmeticComponent))
  (bind _returnvalue (Class|FortPlayerCompanionCosmeticComponent|GetReactionChooserProxy _cosmeticcomponent))
  (Utilities|IsValid CallbackWrapper
    (:Is%20Valid
      (Utilities|Array|Add (Variables|Default|GetPendingReactionCompletedCallbacks) CallbackWrapper))
    (:Is%20Not%20Valid
      _))
  (Utilities|IsValid _cosmeticcomponent
    (:Is%20Valid
      (Utilities|IsValid _returnvalue
        (:Is%20Valid
          (bind _result (Animation|EvaluateProxy (Utilities|Struct|MakeFortPlayerCompanionReactionChooserInput (GameplayTags|MakeGameplayTagContainerfromTag ReactionTag)) _returnvalue)
            (:execute
              (Utilities|IsValid _result
                (:Is%20Valid
                  (Animation|Montage|PlayMontage (Variables|FortPlayerCompanionPreviewActor|GetMesh) _result
                    (:OnCompleted
                      _)
                    (:OnInterrupted
                      _)))
                (:Is%20Not%20Valid
                  _))
              (Utilities|IsValid CallbackWrapper
                (:Is%20Valid
                  (Utilities|Array|RemoveItem (Variables|Default|GetPendingReactionCompletedCallbacks) CallbackWrapper)
                  (Class|FortPlayerCompanionPreviewActorReactionFinishedDelegateWrapper|ExecuteReactionFinished CallbackWrapper))))))
        (:Is%20Not%20Valid
          _)))
    (:Is%20Not%20Valid
      _)))

(event EventOnFrontendCameraChanged
  (params (NewCamera :type EFrontEndCamera%20Enum :default "Invalid") (OldCamera :type EFrontEndCamera%20Enum :default "Invalid"))
  (Variables|Default|SetDiscoveryHomeHidden (Utilities|Enum|Equal(Enum) NewCamera "DiscoveryHome"))
  (CallFunction|RefreshVisibility))

(event EventOnMemberGameReadinessChanged
  (params (NewGameReadiness :type EGameReadiness%20Enum :default "NotReady"))
  (Variables|Default|SetOwnerGameReadinessHidden (Utilities|Select :Index NewGameReadiness :NotReady false :Ready false :Reevaluating false :SittingOut true :Away true))
  (CallFunction|RefreshVisibility))

(event EventOnMemberIsOnFrontendChanged
  (params (bIsInFrontend :type Boolean))
  (bind _companion (Variables|Default|GetCompanion))
  (bind _returnvalue (Rendering|Material|GetNumMaterials _companion))
  (Class|FortPlayerCompanionCosmeticComponent|SetOverrideMaterial (Variables|FortPlayerCompanionPreviewActor|GetCosmeticComponent) (Selection|SelectMaterialInstance 0 (Variables|Default|GetHologramMat) bIsInFrontend) (Variables|Default|GetMaterialParamsToCopy) true)
  (bind _returnvalue_2 (CallFunction|GetAllSkeletalMeshComponents))
  (Utilities|IsValid (Utilities|Array|Get(acopy) _returnvalue_2 1)
    (:Is%20Valid
      (Variables|Default|SetCompanion (Utilities|Array|Get(acopy) _returnvalue_2 1))
      (Utilities|DoOnceThenDo false
        (:DoOnce
          (for _index (range (+ _returnvalue 1))
            ; Store the original overlay to reapply it back once off hologram
            (Utilities|FlowControl|Sequence
              (:then_0
                (Utilities|IsValid (Rendering|Material|GetOverlayMaterial _companion false _index :!class /Script/Engine.MeshComponent)
                  (:Is%20Valid
                    (Variables|Default|SetMaterialSlotsOverlay false))
                  (:Is%20Not%20Valid
                    (Variables|Default|SetMaterialSlotsOverlay true))))
              (:then_1
                (Utilities|Array|Add (Variables|Default|GetOverlay%20Materials) (Rendering|Material|GetOverlayMaterial _companion (Variables|Default|GetMaterialSlotsOverlay) _index :!class /Script/Engine.MeshComponent))))))
        (:ThenDo
          _))
      ; Remove overlay material for holo
      (for _index_2 (range (+ _returnvalue 1))
        (bind _asmaterial_instance_dynamic (Utilities|Casting|CastToMaterialInstanceDynamic (Rendering|Material|GetMaterial _companion 0)))
        (if (== (Variables|/Script/Engine.MaterialInstance|GetParent _asmaterial_instance_dynamic) "/Game/UI/Frontend/Lobby/HeroSquad/M_CharacterShader_Holo_Inst.M_CharacterShader_Holo_Inst")
          (Rendering|Material|SetOverlayMaterial _companion "/Game/Global/Materials/M_Overlay_VertexCrunch.M_Overlay_VertexCrunch" true _index_2 :!class /Script/Engine.MeshComponent)
          (else
            ; Revert back to original overlay material
            (for (_array_element _array_index) (Variables|Default|GetOverlay%20Materials)
              (Rendering|Material|SetOverlayMaterial _companion _array_element true _array_index :!class /Script/Engine.MeshComponent)
              (Rendering|Material|SetOverlayMaterial _companion _array_element (Variables|Default|GetMaterialSlotsOverlay) _array_index :!class /Script/Engine.MeshComponent)))))
      ; Ensure outlines are refresh as it won't set without it.
      (CallFunction|RefreshVisibility))))

(event EventPlayResInEffect
  (bind (_returnvalue_location _returnvalue_rotation) (Math|Transform|BreakTransform (Transformation|GetActorTransform)))
  (Niagara|SpawnSystematLocation "/Game/Effects/Fort_Effects/UI/Transition/NS_FrontEndCompanionTransition.NS_FrontEndCompanionTransition" _returnvalue_location _returnvalue_rotation "1.000000,1.000000,0.500000")
  (bind _returnvalue (CallFunction|GetAllSkeletalMeshComponents))
  (bind (_direction _hiddenamount) (Timeline|Timeline PlayFromStart)
    (:Update
      (for _array_element _returnvalue
        (Utilities|IsValid _array_element
          (:Is%20Valid
            (Rendering|Material|SetScalarParameterValueonMaterials _array_element "NewDissolveGradient" _hiddenamount)
            (Audio|PlaySound2D "/Game/Sounds/Fort_UI_Sounds/S13_Frontend/BattlePass_ResIn_Cue.BattlePass_ResIn_Cue")))))
    (:ShowVFXParticles
      ;[1 1 1] Show the VFX Hidden when the customization was ready
      (Class|FortPlayerCompanionCosmeticComponent|SetCosmeticsVFXVisiblity (Variables|FortPlayerCompanionPreviewActor|GetCosmeticComponent) true))))

(fn IsCompanionHidden
  (return (Variables|Default|GetIsHidden)))

(fn ShowMesh
  (params (Show :type Boolean))
  (bind _mesh (Variables|FortPlayerCompanionPreviewActor|GetMesh))
  (bind _returnvalue_location (Math|Transform|BreakTransform (Transformation|GetRelativeTransform _mesh)))
  (if Show
    (Transformation|SetRelativeLocation _mesh (Variables|Default|GetCachedMeshLocation) false false)
    (else
      (Variables|Default|SetCachedMeshLocation _returnvalue_location)
      (Transformation|SetRelativeLocation _mesh "0.000000,0.000000,-200.000000" false false))))

(fn ConstructionScript)
